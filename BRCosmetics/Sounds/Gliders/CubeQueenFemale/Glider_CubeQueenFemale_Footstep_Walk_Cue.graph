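# Sound cue for the Cube Queen Female glider's walking footstep, using the glider-landing sound class and the glider open/close attenuation.
# A Random node picks one of two equally weighted footstep waves, and a Modulator varies its pitch and volume, with a wide volume range.
# A Quality Level root feeds that chain to the first two quality slots and leaves the other two unconnected.
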SoundCue /BRCosmetics/Sounds/Gliders/CubeQueenFemale/Glider_CubeQueenFemale_Footstep_Walk_Cue
root=QualityLevel_0
nodes [
 {id=Modulator_0, children [Random_0], properties {pitchMax=1.1, volumeMax=1, volumeMin=0.2}}
 {id=QualityLevel_0, children [Modulator_0,Modulator_0,null,null]}
 {id=Random_0, children [WavePlayer_0,WavePlayer_1], properties {weights [1,1]}}
 {id=WavePlayer_0, properties {soundWaveAssetPtr=&/Glider_CubeQueenFemale_Footstep_Walk_01}}
 {id=WavePlayer_1, properties {soundWaveAssetPtr=&/Glider_CubeQueenFemale_Footstep_Walk_02}}
]
properties {
 attenuationSettings=/Game/Sounds/Attenuation/Fort_Player_Glider_OpenClose_Athena_Att
 maxDistance=10000
 soundClassObject=&Fort_Gliders_Land
 volumeMultiplier=1
}
