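# Sound cue for the thrust sound of the female Cube Queen glider cosmetic.
# A Random node picks one of four equally weighted thrust waves and is excluded from branch culling, then feeds a Modulator at default settings for slight variation.
# The cue uses the shared glider thrust attenuation and concurrency, a glider loop audio properties sheet, and the glider land sound class at reduced volume.
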SoundCue /BRCosmetics/Sounds/Gliders/CubeQueenFemale/Glider_CubeQueenFemale_Thrust_Cue
root=Modulator_2
nodes [
 {id=Modulator_2, children [Random_1]}
 {
  id=Random_1
  children [WavePlayer_1, WavePlayer_2, WavePlayer_3, WavePlayer_11]
  properties {bShouldExcludeFromBranchCulling=true, weights [1,1,1,1]}
 }
 {id=WavePlayer_1, properties {soundWaveAssetPtr=&/Glider_CubeQueenFemale_Thrust_01}}
 {id=WavePlayer_11, properties {soundWaveAssetPtr=&/Glider_CubeQueenFemale_Thrust_04}}
 {id=WavePlayer_2, properties {soundWaveAssetPtr=&/Glider_CubeQueenFemale_Thrust_02}}
 {id=WavePlayer_3, properties {soundWaveAssetPtr=&/Glider_CubeQueenFemale_Thrust_03}}
]
properties {
 attenuationSettings=&Fort_Player_Glider_Thrust_Athena_Att
 audioPropertiesSheet=&APS_Cosmetic_Glider_Loop
 maxDistance=12500
 soundClassObject=&Fort_Gliders_Land
 volumeMultiplier=0.4
 concurrencySet [&Athena_Glider_Thrust_Conc]
}
